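# Child animation Blueprint for the monster head of the male medium Slurp Soldier cosmetic, on the Fortnite base head skeleton.
# Each animation update runs the parent logic, then gets a dynamic material instance from the part's skeletal mesh.
# It converts the forward vector of the neck_01 socket rotation to a color and writes it to the NeckForwardVector parameter, so the material follows neck orientation.
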AnimBlueprint /BRCosmetics/Characters/Player/Male/Medium/Heads/M_MED_Soldier_Slurp/Meshes/Male_Medium_SoldierSlurp_Monster_Head_AnimBP_Child
target_skeleton=/Game/Characters/Player/Common/Fortnite_Base_Head/Fortnite_Base_Head_Export_Skeleton
graphs [EventGraph {nodes [
   n1 = Event BlueprintUpdateAnimation()
   > Parent: BlueprintUpdateAnimation!(execute, DeltaTimeX:float=@n1.DeltaTimeX) -> (then) {is=title K2Node_CallParentFunction}
   > n2 = Get Dynamic Material Instance(SkeletalMeshComponent=GetPartSkeletalMeshComponent())
   n3 = GetSocketRotation!(self:Scene Component Object Reference=GetPartSkeletalMeshComponent(), InSocketName:Name=neck_01) -> (ReturnValue:Rotator)
   > SetVectorParameterValue(
    self=@n2
    ParameterName=NeckForwardVector
    Value="To LinearColor (Vector)"!(InVec:Vector=GetForwardVector!(InRot:Rotator=@n3) -> (ReturnValue:Vector)) -> (ReturnValue:Linear Color Structure)
   )
   TryGetPawnOwner()
  ]}]
variables []
