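# Physics asset for the head mesh of the male medium Slurp Soldier character, setting up physics for two dynamic hair chains, bigHair and littleHair.
# Each chain links a tiny base capsule to two larger capsules along the hair bones.
# Both bases constrain to a large head sphere with free swing and twist.
# The joints further along use narrow angular limits, with swing slightly wider at the tips.
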PhysicsAsset /BRCosmetics/Characters/Player/Male/Medium/Heads/M_MED_Soldier_Slurp/Meshes/Male_Medium_Soldier_Slurp_Head_Head_slurp_normal_LOD0_Physics
bodies {
 dyn_bigHair_01 [{shapeType=Capsule, length=2.6527, radius=1.7099, center=<1.7619 0 0>, rotation=<rot 90 0 0>}]
 dyn_bigHair_01_Base [{shapeType=Capsule, length=0.505, radius=0.505, rotation=<rot 0 -90 0>}]
 dyn_bigHair_02 [{shapeType=Capsule, length=2.535, radius=1.6084, center=<2.0311 0 0>, rotation=<rot 90 0 0>}]
 dyn_littleHair_01 [{shapeType=Capsule, length=1.791, radius=1.0674, center=<1.3549 0 -0.0003>, rotation=<rot 90 0 0>}]
 dyn_littleHair_01_Base [{shapeType=Capsule, length=0.505, radius=0.505, rotation=<rot 0 -90 0>}]
 dyn_littleHair_02 [{shapeType=Capsule, length=1.329, radius=0.9333, center=<1.227 0 0>, rotation=<rot 90 0 0>}]
 head [{shapeType=Sphere, radius=9.2335, center=<4.7607 2.0918 0>}]
}
constraints [
 {
  bone1Name=dyn_bigHair_01_Base
  bone2Name=head
  swing1LimitDegrees=45
  swing1Motion=Free
  swing2LimitDegrees=45
  swing2Motion=Free
  twistLimitDegrees=45
  twistMotion=Free
 }
 {bone1Name=dyn_bigHair_01, bone2Name=dyn_bigHair_01_Base, swing1LimitDegrees=12, swing2LimitDegrees=12, twistLimitDegrees=12}
 {bone1Name=dyn_bigHair_02, bone2Name=dyn_bigHair_01, swing1LimitDegrees=16, swing2LimitDegrees=16, twistLimitDegrees=12}
 {
  bone1Name=dyn_littleHair_01_Base
  bone2Name=head
  swing1LimitDegrees=45
  swing1Motion=Free
  swing2LimitDegrees=45
  swing2Motion=Free
  twistLimitDegrees=45
  twistMotion=Free
 }
 {bone1Name=dyn_littleHair_01, bone2Name=dyn_littleHair_01_Base, swing1LimitDegrees=12, swing2LimitDegrees=12, twistLimitDegrees=12}
 {bone1Name=dyn_littleHair_02, bone2Name=dyn_littleHair_01, swing1LimitDegrees=16, swing2LimitDegrees=16, twistLimitDegrees=12}
]
